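# Sound cue for the final wood impact in demolition, routed to the Fort impacts sound class with concurrency limiting.
# A Quality Level root picks one of four mixers, with random wave variety shrinking from the first branch to single waves in the last two.
# Each mixer layers close, medium and far sounds through distance-crossfade attenuations.
# The close layer is briefly delayed and pitched up, the medium layer is modulated, and the far layer plays short small impacts.
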SoundCue /Figment_S06/Sounds/Fort_Impact_Sounds/Demolition/Defaults/Demo_Wood_ImpactFinal_Cue
root=QualityLevel_0
nodes [
 {id=Attenuation_0, children [Random_4], properties {attenuationSettings=/Figment_S06/Sounds/Attenuation/Fort_Demolition_Far_Crossfade_Att}}
 {
  id=Attenuation_1
  children [Modulator_2]
  properties {attenuationSettings=/Figment_S06/Sounds/Attenuation/Fort_Demolition_Medium_Crossfade_Att}
 }
 {
  id=Attenuation_10
  children [Modulator_7]
  properties {attenuationSettings=&Fort_Demolition_Medium_Crossfade_Att}
 }
 {id=Attenuation_11, children [Delay_3], properties {attenuationSettings=/Figment_S06/Sounds/Attenuation/Fort_Demolition_Close_Crossfade_Att}}
 {id=Attenuation_2, children [Random_0], properties {attenuationSettings=&Fort_Demolition_Far_Crossfade_Att}}
 {id=Attenuation_3, children [Delay_0], properties {attenuationSettings=&Fort_Demolition_Close_Crossfade_Att}}
 {
  id=Attenuation_4
  children [WavePlayer_21]
  properties {attenuationSettings=&Fort_Demolition_Far_Crossfade_Att}
 }
 {
  id=Attenuation_5
  children [Modulator_1]
  properties {attenuationSettings=&Fort_Demolition_Medium_Crossfade_Att}
 }
 {id=Attenuation_6, children [Delay_2], properties {attenuationSettings=&Fort_Demolition_Close_Crossfade_Att}}
 {
  id=Attenuation_7
  children [Modulator_5]
  properties {attenuationSettings=&Fort_Demolition_Medium_Crossfade_Att}
 }
 {id=Attenuation_8, children [Delay_1], properties {attenuationSettings=&Fort_Demolition_Close_Crossfade_Att}}
 {
  id=Attenuation_9
  children [WavePlayer_25]
  properties {attenuationSettings=&Fort_Demolition_Far_Crossfade_Att}
 }
 {id=Delay_0, children [Modulator_0], properties {delayMax=0.1, delayMin=0.05}}
 {id=Delay_1, children [Modulator_4], properties {delayMax=0.1, delayMin=0.05}}
 {id=Delay_2, children [Modulator_3], properties {delayMax=0.1, delayMin=0.05}}
 {id=Delay_3, children [Modulator_6], properties {delayMax=0.1, delayMin=0.05}}
 {id=Mixer_0, children [Attenuation_3, Attenuation_1, Attenuation_0], properties {inputVolume [1,1,1]}}
 {id=Mixer_1, children [Attenuation_6, Attenuation_5, Attenuation_2], properties {inputVolume [1,1,1]}}
 {id=Mixer_2, children [Attenuation_8, Attenuation_7, Attenuation_4], properties {inputVolume [1,1,1]}}
 {id=Mixer_3, children [Attenuation_11, Attenuation_10, Attenuation_9], properties {inputVolume [1,1,1]}}
 {id=Modulator_0, children [Random_2], properties {pitchMax=1.3, pitchMin=1.1}}
 {id=Modulator_1, children [Random_1]}
 {id=Modulator_2, children [Random_3]}
 {id=Modulator_3, children [Random_9], properties {pitchMax=1.3, pitchMin=1.1}}
 {id=Modulator_4, children [WavePlayer_16], properties {pitchMax=1.3, pitchMin=1.1}}
 {id=Modulator_5, children [WavePlayer_17]}
 {id=Modulator_6, children [WavePlayer_23], properties {pitchMax=1.3, pitchMin=1.1}}
 {id=Modulator_7, children [WavePlayer_24]}
 {id=QualityLevel_0, children [Mixer_1,Mixer_0,Mixer_2,Mixer_3]}
 {id=Random_0, children [WavePlayer_0,WavePlayer_1,WavePlayer_2,WavePlayer_3], properties {weights [1,1,1,1]}}
 {id=Random_1, children [WavePlayer_89, WavePlayer_105, WavePlayer_106], properties {weights [1,1,1]}}
 {id=Random_2, children [WavePlayer_5,WavePlayer_9,WavePlayer_7], properties {weights [1,1,1]}}
 {id=Random_3, children [WavePlayer_6, WavePlayer_10], properties {weights [1,1]}}
 {id=Random_4, children [WavePlayer_12, WavePlayer_13], properties {bRandomizeWithoutReplacement=false, weights [1,1]}}
 {id=Random_9, children [WavePlayer_88, WavePlayer_94, WavePlayer_91, WavePlayer_93], properties {weights [1,1,1,1]}}
 {id=WavePlayer_0, properties {soundWaveAssetPtr=/Figment_S06/Sounds/Fort_Impact_Sounds/Demolition/Wood_ImpactShort_Small08}}
 {id=WavePlayer_1, properties {soundWaveAssetPtr=/Figment_S06/Sounds/Fort_Impact_Sounds/Demolition/Wood_ImpactShort_Small07}}
 {id=WavePlayer_10, properties {soundWaveAssetPtr=/Figment_S06/Sounds/Fort_Impact_Sounds/Demolition/Fort_Impact_WoodFinal_02}}
 {id=WavePlayer_105, properties {soundWaveAssetPtr=&Fort_Impact_WoodFinal_02}}
 {id=WavePlayer_106, properties {soundWaveAssetPtr=/Figment_S06/Sounds/Fort_Impact_Sounds/Demolition/Fort_Impact_WoodFinal_03}}
 {id=WavePlayer_12, properties {soundWaveAssetPtr=&Wood_ImpactShort_Small08}}
 {id=WavePlayer_13, properties {soundWaveAssetPtr=&Wood_ImpactShort_Small07}}
 {id=WavePlayer_16, properties {soundWaveAssetPtr=/Figment_S06/Sounds/Fort_Impact_Sounds/Demolition/wood_impact_stereo_01}}
 {id=WavePlayer_17, properties {soundWaveAssetPtr=/Figment_S06/Sounds/Fort_Impact_Sounds/Demolition/Fort_Impact_WoodFinal_01}}
 {id=WavePlayer_2, properties {soundWaveAssetPtr=/Figment_S06/Sounds/Fort_Impact_Sounds/Demolition/Wood_ImpactShort_Small09}}
 {id=WavePlayer_21, properties {soundWaveAssetPtr=&Wood_ImpactShort_Small08}}
 {id=WavePlayer_23, properties {soundWaveAssetPtr=&wood_impact_stereo_01}}
 {id=WavePlayer_24, properties {soundWaveAssetPtr=&Fort_Impact_WoodFinal_01}}
 {id=WavePlayer_25, properties {soundWaveAssetPtr=&Wood_ImpactShort_Small08}}
 {id=WavePlayer_3, properties {soundWaveAssetPtr=/Figment_S06/Sounds/Fort_Impact_Sounds/Demolition/Wood_ImpactShort_Small10}}
 {id=WavePlayer_5, properties {soundWaveAssetPtr=&wood_impact_stereo_01}}
 {id=WavePlayer_6, properties {soundWaveAssetPtr=&Fort_Impact_WoodFinal_01}}
 {id=WavePlayer_7, properties {soundWaveAssetPtr=/Figment_S06/Sounds/Fort_Impact_Sounds/Demolition/wood_impact_stereo_07}}
 {id=WavePlayer_88, properties {soundWaveAssetPtr=&wood_impact_stereo_01}}
 {id=WavePlayer_89, properties {soundWaveAssetPtr=&Fort_Impact_WoodFinal_01}}
 {id=WavePlayer_9, properties {soundWaveAssetPtr=/Figment_S06/Sounds/Fort_Impact_Sounds/Demolition/wood_impact_stereo_04}}
 {id=WavePlayer_91, properties {soundWaveAssetPtr=&wood_impact_stereo_07}}
 {id=WavePlayer_93, properties {soundWaveAssetPtr=/Figment_S06/Sounds/Fort_Impact_Sounds/Demolition/wood_impact_stereo_08}}
 {id=WavePlayer_94, properties {soundWaveAssetPtr=&wood_impact_stereo_04}}
]
properties {
 maxDistance=4500
 soundClassObject=&Fort_Impacts_PSM
 volumeMultiplier=0.625
 concurrencyOverrides {resolutionRule=PreventNew}
 concurrencySet [&CONC_Object_Impacts_FourMax]
}
